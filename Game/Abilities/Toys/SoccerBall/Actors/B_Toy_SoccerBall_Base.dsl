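; Base soccer ball toy that measures how far the ball travels and reports it to squad and quests.
; On hit or water entry, each gated by its own DoOnce, the server converts 2D distance from spawn to rounded meters and sends a quest stat.
; The distance variable's OnRep prints the owner's message on remote clients.
; A scoring helper plays a target or net cue, then silently destroys the ball.
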(event Collision|EventHit
  (params (MyComp :type Primitive%20Component%20Object%20Reference) (Other :type Actor%20Object%20Reference) (OtherComp :type Primitive%20Component%20Object%20Reference) (bSelfMoved :type Boolean) (HitLocation :type Vector :default "0, 0, 0") (HitNormal :type Vector :default "0, 0, 0") (NormalImpulse :type Vector :default "0, 0, 0") (Hit :type Hit%20Result%20Structure%20%28by%20ref%29))
  ;[1 1 1] When the ball lands, calculate how far it traveled and display a message. Calculate the distance on the server, then pass off to clients for message displays
  (|CallParentFunction MyComp Other OtherComp bSelfMoved HitLocation HitNormal NormalImpulse Hit)
  (Utilities|FlowControl|DoOnce false
    (:Completed
      (if (Networking|IsServer)
        (bind _rawdistance (CallFunction|CalcRawDistanceTraveled (Transformation|GetActorLocation)))
        (bind _output_get (Variables|Default|SetRawDistanceTraveled _rawdistance))
        (bind _distanceinmeters (CallFunction|ConvertDistancetoMetersRounded _output_get))
        (bind _output_get_2 (Variables|Default|SetDistanceTraveledInMeters _distanceinmeters))
        ;[1 1 1] Send Quest Update
        (Quest|SendComplexCustomStatEvent (Quests|GetQuestManager (Fortnite|GetFortPlayerControllerfromActor (Actor|GetOwner)) "Invalid") 0 "(GameplayTags=((TagName=\"Athena.Quests.SoccerBallToy.Distance\")))" "(GameplayTags=)" _output_get_2)))))

(event OnEnterWaterEvent(WaterInteractionComponent)
  (params (WaterBody :type Fort%20Water%20Body%20Actor%20Object%20Reference) (WaterInteractionComponent :type Fort%20Water%20Interaction%20Component%20Object%20Reference) (bIsFirstBody :type Boolean))
  ;[1 1 1] When the ball lands, calculate how far it traveled and display a message. Calculate the distance on the server, then pass off to clients for message displays
  (Utilities|FlowControl|DoOnce false
    (:Completed
      (if (Networking|IsServer)
        (bind _rawdistance (CallFunction|CalcRawDistanceTraveled (Transformation|GetActorLocation)))
        (bind _output_get (Variables|Default|SetRawDistanceTraveled _rawdistance))
        (bind _distanceinmeters (CallFunction|ConvertDistancetoMetersRounded _output_get))
        (bind _output_get_2 (Variables|Default|SetDistanceTraveledInMeters _distanceinmeters))
        ;[1 1 1] Send Quest Update
        (Quest|SendComplexCustomStatEvent (Quests|GetQuestManager (Fortnite|GetFortPlayerControllerfromActor (Actor|GetOwner)) "Invalid") 0 "(GameplayTags=((TagName=\"Athena.Quests.SoccerBallToy.Distance\")))" "(GameplayTags=)" _output_get_2)))))

(fn CalcRawDistanceTraveled
  (params (FinalPosition :type Vector))
  (return (Math|Vector|Distance2D(Vector) FinalPosition (Variables|Default|GetInitialSpawnPosition))))

(fn ConvertDistanceToMetersRounded
  (params (RawDistance :type Float%20%28double-precision%29))
  (return (Math|Float|Round (/ RawDistance 1e+02))))

(fn DisplayDistanceMessage
  (params (Distance :type Integer))
  (bind _returnvalue (Actor|GetOwner))
  (bind (_asfort_player_state _bsuccess) (Utilities|Casting|CastToFortPlayerState _returnvalue))
  (bind _returnvalue_2 (BlueprintContext|GetFortToyContext))
  (if _bsuccess
    (Utilities|IsValid _returnvalue_2
      (:Is%20Valid
        (bind _returnvalue_3 (Class|FortToyContext|FormatPlayerNameforDisplay _returnvalue_2 _asfort_player_state "ToyOwner"))
        (Class|FortToyContext|ConditionalPrinttoMessageFeed _returnvalue_2 _asfort_player_state (Transformation|GetActorLocation) "SquadMembersOnly" (Utilities|Text|FormatText (Variables|Default|GetDistanceMessageFormat) _returnvalue_3 Distance)))
      (:Is%20Not%20Valid
        (Development|PrintString "Fort Toy Context invalid")))))

(fn OnRep_DistanceTraveledInMeters
  (Networking|SwitchHasAuthority
    (:Remote
      (CallFunction|DisplayDistanceMessage (Variables|Default|GetDistanceTraveledInMeters)))))

(fn PlayScoringCueAndDestroy
  (params (bHitTarget :type Boolean))
  (CallFunction|PlayCueCenteredonMe (select bHitTarget (Variables|Default|GetCue_HitTarget) (Variables|Default|GetCue_HitNet)))
  (Gameplay|SilentDie false))

(fn ConstructionScript
  (|CallParentFunction))
